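# Physics asset for the medium Commando wizard hat face accessory, setting up physical motion for its beard and hood bones.
# Single capsule bodies cover the head, jaw, upper spine, chin beard bone and seven hood bones around the back, sides and front.
# Hood bones constrain to the head with narrow swing limits, tightest at the front top, while the beard swings more widely from the jaw.
# Jaw-to-head and head-to-spine constraints leave swing and twist free.
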PhysicsAsset /BRCosmetics/Accessories/Hats/M_MED_Commando_Wizard_Hat/Meshes/M_MED_Wizard_FaceAcc_Dyn
bodies {
 C_jaw [{shapeType=Capsule, length=2, radius=2}]
 dyn_beard_chin [{shapeType=Capsule, length=4.4597, radius=3.4932, center=<2.2195 0.0001 0>, rotation=<rot 74.9851 -180 -180>}]
 dyn_hood_bk [{shapeType=Capsule, length=3, radius=3, center=<-2.5483 0 0>, rotation=<rot 90 0 0>}]
 dyn_hood_fr_l [{shapeType=Capsule, length=3, radius=3, center=<2.1068 0 -0.0001>, rotation=<rot -90 -0.0396 0.0397>}]
 dyn_hood_fr_r [{shapeType=Capsule, length=3, radius=3, center=<-2.1068 0 0.0001>, rotation=<rot -90 0 0>}]
 dyn_hood_fr_top_l [{shapeType=Capsule, length=3, radius=3, center=<0 -0.2 0>, rotation=<rot 0 0 90>}]
 dyn_hood_fr_top_r [{shapeType=Capsule, length=3, radius=3, center=<0 0.2 0>, rotation=<rot 0 0 90>}]
 dyn_hood_l [{shapeType=Capsule, length=3, radius=3, center=<2.5 0 0>, rotation=<rot 90 0 0>}]
 dyn_hood_r [{shapeType=Capsule, length=3, radius=3, center=<-2.5 0 0>, rotation=<rot 90 0 0>}]
 head [{shapeType=Capsule, length=5, radius=5}]
 spine_05 [{shapeType=Capsule, length=16, radius=10.25, center=<3.1582 0.4687 0>}]
}
constraints [
 {bone1Name=dyn_beard_chin, bone2Name=C_jaw, swing1LimitDegrees=60, swing2LimitDegrees=60, twistLimitDegrees=5}
 {bone1Name=dyn_hood_bk, bone2Name=head, swing1LimitDegrees=20, swing2LimitDegrees=20, twistLimitDegrees=5}
 {bone1Name=dyn_hood_l, bone2Name=head, swing1LimitDegrees=15, swing2LimitDegrees=15, twistLimitDegrees=5}
 {bone1Name=dyn_hood_fr_l, bone2Name=head, swing1LimitDegrees=15, swing2LimitDegrees=15, twistLimitDegrees=5}
 {bone1Name=dyn_hood_r, bone2Name=head, swing1LimitDegrees=15, swing2LimitDegrees=15, twistLimitDegrees=5}
 {bone1Name=dyn_hood_fr_r, bone2Name=head, swing1LimitDegrees=15, swing2LimitDegrees=15, twistLimitDegrees=5}
 {bone1Name=dyn_hood_fr_top_l, bone2Name=head, swing1LimitDegrees=5, swing2LimitDegrees=5, twistLimitDegrees=5}
 {bone1Name=dyn_hood_fr_top_r, bone2Name=head, swing1LimitDegrees=5, swing2LimitDegrees=5, twistLimitDegrees=5}
 {
  bone1Name=head
  bone2Name=spine_05
  swing1LimitDegrees=45
  swing1Motion=Free
  swing2LimitDegrees=45
  swing2Motion=Free
  twistLimitDegrees=45
  twistMotion=Free
 }
 {
  bone1Name=C_jaw
  bone2Name=head
  swing1LimitDegrees=45
  swing1Motion=Free
  swing2LimitDegrees=45
  swing2Motion=Free
  twistLimitDegrees=45
  twistMotion=Free
 }
]
